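# Looping ambient spot cue for a Secret Facility computer, with small-spot attenuation and limited concurrency.
# Mixes two layers, idling and calculating, each randomly picking one of nine looping waves.
# Each layer runs through two nested, endlessly looping Envelopers, a 30-second pitch curve and a 20-second volume dip, then a Modulator.
# The differing envelope periods keep pitch and volume movement out of step, varying the static loops.
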SoundCue /Hera/Audio/POI/SecretFacility/Spots/SC_Spot_SecretFacility_Computer_Loop
root=Mixer_0
nodes [
 {
  id=Enveloper_0
  children [Random_2]
  properties {
   bLoop=true
   bLoopIndefinitely=true
   loopEnd=30
   pitchCurve {editorCurveData {keys [
      {arriveTangent=-0.0013, interpMode=RCIM_Cubic, leaveTangent=-0.0013, time=0}
      <at 15 0.98 RCIM_Cubic>
      {arriveTangent=0.0013, interpMode=RCIM_Cubic, leaveTangent=0.0013, time=30}
     ]}}
   volumeCurve {editorCurveData {keys [<at 0 1>]}}
  }
 }
 {
  id=Enveloper_1
  children [Enveloper_0]
  properties {
   bLoop=true
   bLoopIndefinitely=true
   loopEnd=20
   pitchCurve {editorCurveData {keys [{time=0}]}}
   volumeCurve {editorCurveData {keys [
      {arriveTangent=-0.01, interpMode=RCIM_Cubic, time=0, value=1}
      <at 10 0.9 RCIM_Cubic>
      {interpMode=RCIM_Cubic, leaveTangent=0.01, time=20, value=1}
     ]}}
  }
 }
 {
  id=Enveloper_6
  children [Random_0]
  properties {
   bLoop=true
   bLoopIndefinitely=true
   loopEnd=30
   pitchCurve {editorCurveData {keys [
      {arriveTangent=-0.0013, interpMode=RCIM_Cubic, leaveTangent=-0.0013, time=0}
      <at 15 0.98 RCIM_Cubic>
      {arriveTangent=0.0013, interpMode=RCIM_Cubic, leaveTangent=0.0013, time=30}
     ]}}
   volumeCurve {editorCurveData {keys [<at 0 1>]}}
  }
 }
 {
  id=Enveloper_7
  children [Enveloper_6]
  properties {
   bLoop=true
   bLoopIndefinitely=true
   loopEnd=20
   pitchCurve {editorCurveData {keys [{time=0}]}}
   volumeCurve {editorCurveData {keys [
      {arriveTangent=-0.01, interpMode=RCIM_Cubic, time=0, value=1}
      <at 10 0.9 RCIM_Cubic>
      {interpMode=RCIM_Cubic, leaveTangent=0.01, time=20, value=1}
     ]}}
  }
 }
 {id=Mixer_0, children [Modulator_0,Modulator_1], properties {inputVolume [1,1]}}
 {id=Modulator_0, children [Enveloper_7]}
 {id=Modulator_1, children [Enveloper_1]}
 {
  id=Random_0
  children [WavePlayer_0,WavePlayer_1,WavePlayer_2,WavePlayer_3,WavePlayer_4,WavePlayer_5,WavePlayer_6,WavePlayer_7,WavePlayer_8]
  properties {weights [1,1,1,1,1,1,1,1,1]}
 }
 {
  id=Random_2
  children [WavePlayer_9, WavePlayer_18, WavePlayer_19, WavePlayer_20, WavePlayer_21, WavePlayer_22, WavePlayer_23, WavePlayer_24, WavePlayer_25]
  properties {weights [1,1,1,1,1,1,1,1,1]}
 }
 {
  id=WavePlayer_0
  properties {bLooping=true, soundWaveAssetPtr=&/Spot_SecretFacility_Computer_Idling_Loop_01}
 }
 {
  id=WavePlayer_1
  properties {bLooping=true, soundWaveAssetPtr=&/Spot_SecretFacility_Computer_Idling_Loop_02}
 }
 {
  id=WavePlayer_18
  properties {bLooping=true, soundWaveAssetPtr=&/Spot_SecretFacility_Computer_Calculating_Loop_02}
 }
 {
  id=WavePlayer_19
  properties {bLooping=true, soundWaveAssetPtr=&/Spot_SecretFacility_Computer_Calculating_Loop_03}
 }
 {
  id=WavePlayer_2
  properties {bLooping=true, soundWaveAssetPtr=&/Spot_SecretFacility_Computer_Idling_Loop_03}
 }
 {
  id=WavePlayer_20
  properties {bLooping=true, soundWaveAssetPtr=&/Spot_SecretFacility_Computer_Calculating_Loop_04}
 }
 {
  id=WavePlayer_21
  properties {bLooping=true, soundWaveAssetPtr=&/Spot_SecretFacility_Computer_Calculating_Loop_05}
 }
 {
  id=WavePlayer_22
  properties {bLooping=true, soundWaveAssetPtr=&/Spot_SecretFacility_Computer_Calculating_Loop_06}
 }
 {
  id=WavePlayer_23
  properties {bLooping=true, soundWaveAssetPtr=&/Spot_SecretFacility_Computer_Calculating_Loop_07}
 }
 {
  id=WavePlayer_24
  properties {bLooping=true, soundWaveAssetPtr=&/Spot_SecretFacility_Computer_Calculating_Loop_08}
 }
 {
  id=WavePlayer_25
  properties {bLooping=true, soundWaveAssetPtr=&/Spot_SecretFacility_Computer_Calculating_Loop_09}
 }
 {
  id=WavePlayer_3
  properties {bLooping=true, soundWaveAssetPtr=&/Spot_SecretFacility_Computer_Idling_Loop_04}
 }
 {
  id=WavePlayer_4
  properties {bLooping=true, soundWaveAssetPtr=&/Spot_SecretFacility_Computer_Idling_Loop_05}
 }
 {
  id=WavePlayer_5
  properties {bLooping=true, soundWaveAssetPtr=&/Spot_SecretFacility_Computer_Idling_Loop_06}
 }
 {
  id=WavePlayer_6
  properties {bLooping=true, soundWaveAssetPtr=&/Spot_SecretFacility_Computer_Idling_Loop_07}
 }
 {
  id=WavePlayer_7
  properties {bLooping=true, soundWaveAssetPtr=&/Spot_SecretFacility_Computer_Idling_Loop_08}
 }
 {
  id=WavePlayer_8
  properties {bLooping=true, soundWaveAssetPtr=&/Spot_SecretFacility_Computer_Idling_Loop_09}
 }
 {
  id=WavePlayer_9
  properties {bLooping=true, soundWaveAssetPtr=&/Spot_SecretFacility_Computer_Calculating_Loop_01}
 }
]
properties {
 attenuationSettings=/Hera/Audio/Attenuations/ATT_Hera_Spot_Small
 bEnableBusSends=false
 bEnableSubmixSends=false
 maxDistance=1100
 priority=5
 soundClassObject=&Fort_Ambient_Indoors_Athena
 volumeMultiplier=0.6
 concurrencySet [/Game/Concurrency/AmbientConc_3Max_Spot_Default_Concurrency]
}
