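# Looping ambient spot cue for the UFO vibe at the Secret Facility POI, played at half pitch with huge-spot attenuation.
# A Mixer blends two tonal layers and a quieter noise layer, each passing through nested looping Envelopers and a Modulator.
# The Envelopers add slow pitch dips and volume swells, with differing volume periods per layer so they drift apart.
# Quality switches keep one tonal layer at all four levels and drop the others at the last two.
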SoundCue /Hera/Audio/POI/SecretFacility/Spots/SC_Spot_SecretFacility_UFO_Vibe
root=Mixer_0
nodes [
 {
  id=Enveloper_2
  children [WavePlayer_16]
  properties {
   bLoop=true
   bLoopIndefinitely=true
   loopEnd=30
   pitchCurve {editorCurveData {keys [
      {arriveTangent=-0.0013, interpMode=RCIM_Cubic, leaveTangent=-0.0013, time=0}
      <at 15 0.98 RCIM_Cubic>
      {arriveTangent=0.0013, interpMode=RCIM_Cubic, leaveTangent=0.0013, time=30}
     ]}}
   volumeCurve {editorCurveData {keys [<at 0 1>]}}
  }
 }
 {
  id=Enveloper_3
  children [Enveloper_2]
  properties {
   bLoop=true
   bLoopIndefinitely=true
   loopEnd=16
   pitchCurve {editorCurveData {keys [{time=0}]}}
   volumeCurve {editorCurveData {keys [
      {arriveTangent=-0.01, interpMode=RCIM_Cubic, time=0, value=1}
      <at 8 0.7 RCIM_Cubic>
      {interpMode=RCIM_Cubic, leaveTangent=0.01, time=16, value=1}
     ]}}
  }
 }
 {
  id=Enveloper_4
  children [WavePlayer_17]
  properties {
   bLoop=true
   bLoopIndefinitely=true
   loopEnd=30
   pitchCurve {editorCurveData {keys [
      {arriveTangent=-0.0013, interpMode=RCIM_Cubic, leaveTangent=-0.0013, time=0}
      <at 15 0.98 RCIM_Cubic>
      {arriveTangent=0.0013, interpMode=RCIM_Cubic, leaveTangent=0.0013, time=30}
     ]}}
   volumeCurve {editorCurveData {keys [<at 0 1>]}}
  }
 }
 {
  id=Enveloper_5
  children [Enveloper_4]
  properties {
   bLoop=true
   bLoopIndefinitely=true
   loopEnd=20
   pitchCurve {editorCurveData {keys [{time=0}]}}
   volumeCurve {editorCurveData {keys [
      {arriveTangent=-0.01, interpMode=RCIM_Cubic, time=0, value=1}
      <at 10 0.7 RCIM_Cubic>
      {interpMode=RCIM_Cubic, leaveTangent=0.01, time=20, value=1}
     ]}}
  }
 }
 {
  id=Enveloper_6
  children [WavePlayer_14]
  properties {
   bLoop=true
   bLoopIndefinitely=true
   loopEnd=30
   pitchCurve {editorCurveData {keys [
      {arriveTangent=-0.0013, interpMode=RCIM_Cubic, leaveTangent=-0.0013, time=0}
      <at 15 0.98 RCIM_Cubic>
      {arriveTangent=0.0013, interpMode=RCIM_Cubic, leaveTangent=0.0013, time=30}
     ]}}
   volumeCurve {editorCurveData {keys [<at 0 1>]}}
  }
 }
 {
  id=Enveloper_7
  children [Enveloper_6]
  properties {
   bLoop=true
   bLoopIndefinitely=true
   loopEnd=30
   pitchCurve {editorCurveData {keys [{time=0}]}}
   volumeCurve {editorCurveData {keys [
      {arriveTangent=-0.01, interpMode=RCIM_Cubic, time=0, value=1}
      <at 15 0.7 RCIM_Cubic>
      {interpMode=RCIM_Cubic, leaveTangent=0.01, time=30, value=1}
     ]}}
  }
 }
 {id=Mixer_0, children [QualityLevel_1, QualityLevel_3, QualityLevel_0], properties {inputVolume [1,1,0.6]}}
 {id=Modulator_0, children [Enveloper_7]}
 {id=Modulator_2, children [Enveloper_3]}
 {id=Modulator_3, children [Enveloper_5]}
 {id=QualityLevel_0, children [Modulator_3,Modulator_3,null,null]}
 {id=QualityLevel_1, children [Modulator_0,Modulator_0,null,null]}
 {id=QualityLevel_3, children [Modulator_2,Modulator_2,Modulator_2,Modulator_2]}
 {id=WavePlayer_14, properties {bLooping=true, soundWaveAssetPtr=&/Spot_SecretFacility_UFO_Tonal_Layer_01}}
 {id=WavePlayer_16, properties {bLooping=true, soundWaveAssetPtr=&/Spot_SecretFacility_UFO_Tonal_Layer_03}}
 {
  id=WavePlayer_17
  properties {bLooping=true, soundWaveAssetPtr=&/Spot_SecretFacility_UFO_Vibe_Noise_Layer_01}
 }
]
properties {
 attenuationSettings=/Hera/Audio/Attenuations/ATT_Hera_Spot_Huge
 bEnableBusSends=false
 bEnableSubmixSends=false
 maxDistance=3900
 pitchMultiplier=0.5
 priority=5
 soundClassObject=&Fort_Ambient_Indoors_Athena
 volumeMultiplier=1
 attenuationOverrides {falloffDistance=108476.9531, attenuationShapeExtents {x=632023.7954}}
 concurrencySet [/Game/Concurrency/AmbientConc_3Max_Spot_Default_Concurrency]
}
