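# Looping exterior spot ambience for the SecretFacility POI: gusting wind rattling metal vent slats, in the outdoor ambient sound class.
# A Random node picks one of two level-balanced looping waves, which passes through a short fade-in envelope.
# Nested looping envelopes of different lengths vary pitch and volume to create gusts, then a fixed boost feeds two of four quality slots.
# Overrides set cone attenuation with low-pass and listener focus, plus single-instance concurrency.
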SoundCue /Hera/Audio/POI/SecretFacility/Spots/SC_Spot_SecretFacility_Wind_Gusting_Metal_Vent_Slats_Loop_Ext
root=QualityLevel_1
nodes [
 {
  id=Enveloper_0
  children [Enveloper_4]
  properties {
   bLoop=true
   bLoopIndefinitely=true
   loopEnd=30
   pitchCurve {editorCurveData {keys [{arriveTangent=-0.0033, interpMode=RCIM_Cubic, time=0}, <at 15 1.05 RCIM_Cubic>, {interpMode=RCIM_Cubic, leaveTangent=0.0033, time=30}]}}
   volumeCurve {editorCurveData {keys [<at 0 1>]}}
  }
 }
 {
  id=Enveloper_3
  children [Enveloper_0]
  properties {
   bLoop=true
   bLoopIndefinitely=true
   loopEnd=20
   pitchCurve {editorCurveData {keys [{time=0}]}}
   volumeCurve {editorCurveData {keys [
      {arriveTangent=-0.01, interpMode=RCIM_Cubic, time=0, value=1}
      {arriveTangent=-0.0666, interpMode=RCIM_Cubic, leaveTangent=-0.0666, time=2.0525, value=0.9761}
      {arriveTangent=-0.0158, interpMode=RCIM_Cubic, leaveTangent=-0.0158, time=5.4253, value=0.6385}
      {arriveTangent=-0.0011, interpMode=RCIM_Cubic, leaveTangent=-0.0011, time=10.0347, value=0.85}
      {arriveTangent=0.0151, interpMode=RCIM_Cubic, leaveTangent=0.0151, time=15.6263, value=0.6268}
      {interpMode=RCIM_Cubic, leaveTangent=0.01, time=20, value=1}
     ]}}
  }
 }
 {
  id=Enveloper_4
  children [Random_0]
  properties {pitchCurve {editorCurveData {keys [{time=0}]}}, volumeCurve {editorCurveData {keys [<at 0 0>,<at 0.2 1>]}}}
 }
 {id=Mixer_4, children [WavePlayer_3], properties {inputVolume [1.25]}}
 {id=Mixer_5, children [WavePlayer_0], properties {inputVolume [0.8]}}
 {id=Modulator_1, children [Enveloper_3], properties {volumeMax=1.3, volumeMin=1.3}}
 {id=QualityLevel_1, children [Modulator_1,Modulator_1,null,null]}
 {id=Random_0, children [Mixer_5,Mixer_4], properties {weights [1,1]}}
 {
  id=WavePlayer_0
  properties {bLooping=true, soundWaveAssetPtr=&/Spot_SecretFacility_Wind_Blustering_Rattling_Building_Vents_01}
 }
 {
  id=WavePlayer_3
  properties {bLooping=true, soundWaveAssetPtr=&/Spot_SecretFacility_Wind_Blustering_Rattling_Building_Vents_02}
 }
]
properties {
 bOverrideAttenuation=true
 bOverrideConcurrency=true
 maxDistance=1350
 soundClassObject=&Fort_Ambient_Outdoors_Athena
 volumeMultiplier=1
 attenuationOverrides {
  attenuationShape=Cone
  bAttenuateWithLPF=true
  bEnableFocusInterpolation=true
  bEnableListenerFocus=true
  falloffDistance=1000
  focusAttackInterpSpeed=0.05
  focusReleaseInterpSpeed=0.01
  lPFFrequencyAtMax=3400
  lPFRadiusMax=1000
  lPFRadiusMin=500
  nonFocusVolumeAttenuation=0.4
  attenuationShapeExtents=<350 60 90>
 }
 concurrencyOverrides {maxCount=1, retriggerTime=2, voiceStealReleaseTime=3}
 concurrencySet [/Hermes/Props/SprocketPoppy/Materials/FaceAnim/AmbientConc_3Max_Spot_Concurrency]
}
