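# Sound cue for underground throbbing, deep metal resonance spot ambience at the Secret Facility POI.
# A Random node picks one of six equally weighted waves, feeding two nested, endlessly looping Envelopers and a root Modulator.
# The inner Enveloper slowly dips pitch and the outer dips volume on a shorter cycle, so the movements drift against each other.
# It uses distant-creak attenuation, the indoor ambient one-shot sound class and three-max spot concurrency.
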SoundCue /Hera/Audio/POI/SecretFacility/Spots/SC_Spot_Underground_Throbbing_Deep_Metal_Resonances
root=Modulator_0
nodes [
 {
  id=Enveloper_6
  children [Random_0]
  properties {
   bLoop=true
   bLoopIndefinitely=true
   loopEnd=30
   pitchCurve {editorCurveData {keys [
      {arriveTangent=-0.0013, interpMode=RCIM_Cubic, leaveTangent=-0.0013, time=0}
      <at 15 0.98 RCIM_Cubic>
      {arriveTangent=0.0013, interpMode=RCIM_Cubic, leaveTangent=0.0013, time=30}
     ]}}
   volumeCurve {editorCurveData {keys [<at 0 1>]}}
  }
 }
 {
  id=Enveloper_7
  children [Enveloper_6]
  properties {
   bLoop=true
   bLoopIndefinitely=true
   loopEnd=20
   pitchCurve {editorCurveData {keys [{time=0}]}}
   volumeCurve {editorCurveData {keys [
      {arriveTangent=-0.01, interpMode=RCIM_Cubic, time=0, value=1}
      <at 10 0.9 RCIM_Cubic>
      {interpMode=RCIM_Cubic, leaveTangent=0.01, time=20, value=1}
     ]}}
  }
 }
 {id=Modulator_0, children [Enveloper_7]}
 {
  id=Random_0
  children [WavePlayer_0, WavePlayer_1, WavePlayer_2, WavePlayer_3, WavePlayer_11, WavePlayer_12]
  properties {weights [1,1,1,1,1,1]}
 }
 {id=WavePlayer_0, properties {soundWaveAssetPtr=&/Spot_Underground_Throbbing_Deep_Metal_Resonances_1}}
 {id=WavePlayer_1, properties {soundWaveAssetPtr=&/Spot_Underground_Throbbing_Deep_Metal_Resonances_2}}
 {id=WavePlayer_11, properties {soundWaveAssetPtr=&/Spot_Underground_Throbbing_Deep_Metal_Resonances_5}}
 {id=WavePlayer_12, properties {soundWaveAssetPtr=&/Spot_Underground_Throbbing_Deep_Metal_Resonances_6}}
 {id=WavePlayer_2, properties {soundWaveAssetPtr=&/Spot_Underground_Throbbing_Deep_Metal_Resonances_3}}
 {id=WavePlayer_3, properties {soundWaveAssetPtr=&/Spot_Underground_Throbbing_Deep_Metal_Resonances_4}}
]
properties {
 attenuationSettings=/Game/Sounds/Attenuation/Fort_TileAmbient_OneShot_Distant_Creak_Att
 bEnableBusSends=false
 bEnableSubmixSends=false
 maxDistance=8200
 priority=5
 soundClassObject=/Game/Sounds/FortSoundClassesAndModes/Fort_Ambient_OneShots_Indoors
 volumeMultiplier=1
 concurrencySet [/Game/Concurrency/AmbientConc_3Max_Spot_Default_Concurrency]
}
