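# Sound cue for an indoor ambient spot in the Secret Facility POI, playing wailing, resonant underground metal movement sounds.
# A Random node picks one of seven equally weighted waves and feeds two chained looping Envelopers into a root Modulator.
# The Envelopers' differing loop lengths apply slow, slight dips to pitch and volume that drift against each other.
# Distant-creak attenuation, an indoor one-shot sound class and a shared ambient concurrency limit govern playback.
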SoundCue /Hera/Audio/POI/SecretFacility/Spots/SC_Spot_Underground_Wailing_Resonant_Howling_Metal_Movements_Int
root=Modulator_0
nodes [
 {
  id=Enveloper_6
  children [Random_1]
  properties {
   bLoop=true
   bLoopIndefinitely=true
   loopEnd=30
   pitchCurve {editorCurveData {keys [
      {arriveTangent=-0.0013, interpMode=RCIM_Cubic, leaveTangent=-0.0013, time=0}
      <at 15 0.98 RCIM_Cubic>
      {arriveTangent=0.0013, interpMode=RCIM_Cubic, leaveTangent=0.0013, time=30}
     ]}}
   volumeCurve {editorCurveData {keys [<at 0 1>]}}
  }
 }
 {
  id=Enveloper_7
  children [Enveloper_6]
  properties {
   bLoop=true
   bLoopIndefinitely=true
   loopEnd=20
   pitchCurve {editorCurveData {keys [{time=0}]}}
   volumeCurve {editorCurveData {keys [
      {arriveTangent=-0.01, interpMode=RCIM_Cubic, time=0, value=1}
      <at 10 0.9 RCIM_Cubic>
      {interpMode=RCIM_Cubic, leaveTangent=0.01, time=20, value=1}
     ]}}
  }
 }
 {id=Modulator_0, children [Enveloper_7]}
 {
  id=Random_1
  children [WavePlayer_4, WavePlayer_5, WavePlayer_6, WavePlayer_7, WavePlayer_8, WavePlayer_9, WavePlayer_10]
  properties {weights [1,1,1,1,1,1,1]}
 }
 {
  id=WavePlayer_10
  properties {soundWaveAssetPtr=&/Spot_Underground_Wailing_Resonant_Howling_Metal_Movements_7}
 }
 {
  id=WavePlayer_4
  properties {soundWaveAssetPtr=&/Spot_Underground_Wailing_Resonant_Howling_Metal_Movements_1}
 }
 {
  id=WavePlayer_5
  properties {soundWaveAssetPtr=&/Spot_Underground_Wailing_Resonant_Howling_Metal_Movements_2}
 }
 {
  id=WavePlayer_6
  properties {soundWaveAssetPtr=&/Spot_Underground_Wailing_Resonant_Howling_Metal_Movements_3}
 }
 {
  id=WavePlayer_7
  properties {soundWaveAssetPtr=&/Spot_Underground_Wailing_Resonant_Howling_Metal_Movements_4}
 }
 {
  id=WavePlayer_8
  properties {soundWaveAssetPtr=&/Spot_Underground_Wailing_Resonant_Howling_Metal_Movements_5}
 }
 {
  id=WavePlayer_9
  properties {soundWaveAssetPtr=&/Spot_Underground_Wailing_Resonant_Howling_Metal_Movements_6}
 }
]
properties {
 attenuationSettings=/Game/Sounds/Attenuation/Fort_TileAmbient_OneShot_Distant_Creak_Att
 bEnableBusSends=false
 bEnableSubmixSends=false
 maxDistance=8200
 priority=5
 soundClassObject=/Game/Sounds/FortSoundClassesAndModes/Fort_Ambient_OneShots_Indoors
 volumeMultiplier=1
 concurrencySet [/Game/Concurrency/AmbientConc_3Max_Spot_Default_Concurrency]
}
