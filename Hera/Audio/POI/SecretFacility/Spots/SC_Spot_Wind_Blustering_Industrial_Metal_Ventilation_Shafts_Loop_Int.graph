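# Looping ambient spot Sound Cue of blustering wind in metal ventilation shafts for the Secret Facility POI interior.
# A random node picks one of eight looping stereo waves, feeding two nested endless envelopers that drift pitch and gently dip volume on different cycle lengths.
# A root Modulator adds default variation.
# It uses medium spot attenuation, a three-voice ambient concurrency limit and, despite the Int suffix, the outdoor ambient sound class.
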SoundCue /Hera/Audio/POI/SecretFacility/Spots/SC_Spot_Wind_Blustering_Industrial_Metal_Ventilation_Shafts_Loop_Int
root=Modulator_0
nodes [
 {
  id=Enveloper_6
  children [Random_1]
  properties {
   bLoop=true
   bLoopIndefinitely=true
   loopEnd=30
   pitchCurve {editorCurveData {keys [
      {arriveTangent=-0.0013, interpMode=RCIM_Cubic, leaveTangent=-0.0013, time=0}
      <at 15 0.98 RCIM_Cubic>
      {arriveTangent=0.0013, interpMode=RCIM_Cubic, leaveTangent=0.0013, time=30}
     ]}}
   volumeCurve {editorCurveData {keys [<at 0 1>]}}
  }
 }
 {
  id=Enveloper_7
  children [Enveloper_6]
  properties {
   bLoop=true
   bLoopIndefinitely=true
   loopEnd=20
   pitchCurve {editorCurveData {keys [{time=0}]}}
   volumeCurve {editorCurveData {keys [
      {arriveTangent=-0.01, interpMode=RCIM_Cubic, time=0, value=1}
      <at 10 0.9 RCIM_Cubic>
      {interpMode=RCIM_Cubic, leaveTangent=0.01, time=20, value=1}
     ]}}
  }
 }
 {id=Modulator_0, children [Enveloper_7]}
 {
  id=Random_1
  children [WavePlayer_10, WavePlayer_11, WavePlayer_12, WavePlayer_13, WavePlayer_14, WavePlayer_15, WavePlayer_16, WavePlayer_17]
  properties {weights [1,1,1,1,1,1,1,1]}
 }
 {
  id=WavePlayer_10
  properties {
   bLooping=true
   soundWaveAssetPtr=&/Spot_Wind_Blustering_Industrial_Metal_Ventilation_Shafts_Loop_Stereo_01
  }
 }
 {
  id=WavePlayer_11
  properties {
   bLooping=true
   soundWaveAssetPtr=&/Spot_Wind_Blustering_Industrial_Metal_Ventilation_Shafts_Loop_Stereo_02
  }
 }
 {
  id=WavePlayer_12
  properties {
   bLooping=true
   soundWaveAssetPtr=&/Spot_Wind_Blustering_Industrial_Metal_Ventilation_Shafts_Loop_Stereo_03
  }
 }
 {
  id=WavePlayer_13
  properties {
   bLooping=true
   soundWaveAssetPtr=&/Spot_Wind_Blustering_Industrial_Metal_Ventilation_Shafts_Loop_Stereo_04
  }
 }
 {
  id=WavePlayer_14
  properties {
   bLooping=true
   soundWaveAssetPtr=&/Spot_Wind_Blustering_Industrial_Metal_Ventilation_Shafts_Loop_Stereo_05
  }
 }
 {
  id=WavePlayer_15
  properties {
   bLooping=true
   soundWaveAssetPtr=&/Spot_Wind_Blustering_Industrial_Metal_Ventilation_Shafts_Loop_Stereo_06
  }
 }
 {
  id=WavePlayer_16
  properties {
   bLooping=true
   soundWaveAssetPtr=&/Spot_Wind_Blustering_Industrial_Metal_Ventilation_Shafts_Loop_Stereo_07
  }
 }
 {
  id=WavePlayer_17
  properties {
   bLooping=true
   soundWaveAssetPtr=&/Spot_Wind_Blustering_Industrial_Metal_Ventilation_Shafts_Loop_Stereo_08
  }
 }
]
properties {
 attenuationSettings=/Hera/Audio/Attenuations/ATT_Hera_Spot_Medium
 bEnableBusSends=false
 bEnableSubmixSends=false
 maxDistance=1500
 priority=5
 soundClassObject=&Fort_Ambient_Outdoors_Athena
 volumeMultiplier=1
 concurrencySet [/Game/Concurrency/AmbientConc_3Max_Spot_Default_Concurrency]
}
