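# Sound cue for a large vent spot ambience at the Secret Facility point of interest.
# A Random node with equal weights picks one of five vent loop wave variations and feeds it straight to the output.
# The graph has no looping, modulation or mixing nodes.
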SoundCue /Hera/Audio/POI/SecretFacility/Spots/Spot_SecretFacility_Vent_Large_Loop_Cue
root=Random_0
nodes [
 {id=Random_0, children [WavePlayer_0,WavePlayer_1,WavePlayer_2,WavePlayer_3,WavePlayer_4], properties {weights [1,1,1,1,1]}}
 {id=WavePlayer_0, properties {soundWaveAssetPtr=&/Spot_SecretFacility_Vent_Large_Loop_01}}
 {id=WavePlayer_1, properties {soundWaveAssetPtr=&/Spot_SecretFacility_Vent_Large_Loop_02}}
 {id=WavePlayer_2, properties {soundWaveAssetPtr=&/Spot_SecretFacility_Vent_Large_Loop_03}}
 {id=WavePlayer_3, properties {soundWaveAssetPtr=&/Spot_SecretFacility_Vent_Large_Loop_04}}
 {id=WavePlayer_4, properties {soundWaveAssetPtr=&/Spot_SecretFacility_Vent_Large_Loop_05}}
]
properties {maxDistance=2097152, volumeMultiplier=1}
